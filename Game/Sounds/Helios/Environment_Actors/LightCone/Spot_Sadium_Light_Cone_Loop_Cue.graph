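# Looping outdoor ambient cue for a stadium spotlight's electrical buzz and light cone.
# It mixes a quiet, short-range transformer buzz with a louder light buzz in a long cone, each with its own attenuation node.
# Looping envelopes swell the layers in opposition over 20 seconds.
# A random start, immediate or briefly delayed, plus pitch modulation varies instances, which are capped by concurrency and feed exterior bus and reverb sends.
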SoundCue /Game/Sounds/Helios/Environment_Actors/LightCone/Spot_Sadium_Light_Cone_Loop_Cue
root=Random_0
nodes [
 {
  id=Attenuation_0
  children [Enveloper_0]
  properties {
   bOverrideAttenuation=true
   attenuationOverrides {
    bEnableFocusInterpolation=true
    bEnableListenerFocus=true
    falloffDistance=550
    nonFocusVolumeAttenuation=0.75
    stereoSpread=100
    attenuationShapeExtents {x=25}
   }
  }
 }
 {
  id=Attenuation_1
  children [Enveloper_1]
  properties {
   bOverrideAttenuation=true
   attenuationOverrides {
    attenuationShape=Cone
    bEnableFocusInterpolation=true
    bEnableListenerFocus=true
    coneOffset=45
    distanceAlgorithm=Custom
    falloffDistance=2000
    focusAzimuth=18
    nonFocusAzimuth=45
    nonFocusVolumeAttenuation=0.75
    stereoSpread=100
    attenuationShapeExtents=<1250 20 10>
    customAttenuationCurve {editorCurveData {keys [<at 0.2141 0.6144>,<at 0.5105 0.9957>,<at 0.8 1>,<at 1 0>]}}
   }
  }
 }
 {id=Delay_0, children [Modulator_2], properties {delayMax=0.35, delayMin=0.35}}
 {
  id=Enveloper_0
  children [Modulator_0]
  properties {
   bLoop=true
   bLoopIndefinitely=true
   loopEnd=20
   pitchCurve {editorCurveData {keys [{time=0}]}}
   volumeCurve {editorCurveData {keys [
      {arriveTangent=-0.01, interpMode=RCIM_Cubic, time=0, value=1}
      <at 10 0.7 RCIM_Cubic>
      {interpMode=RCIM_Cubic, leaveTangent=0.01, time=20, value=1}
     ]}}
  }
 }
 {
  id=Enveloper_1
  children [Modulator_1]
  properties {
   bLoop=true
   bLoopIndefinitely=true
   loopEnd=20
   pitchCurve {editorCurveData {keys [{time=0}]}}
   volumeCurve {editorCurveData {keys [
      {arriveTangent=-0.01, interpMode=RCIM_Cubic, time=0, value=1}
      <at 10 1.3 RCIM_Cubic>
      {interpMode=RCIM_Cubic, leaveTangent=0.01, time=20, value=1}
     ]}}
  }
 }
 {id=Mixer_0, children [Attenuation_0, Attenuation_1], properties {inputVolume [0.25,3]}}
 {id=Modulator_0, children [WavePlayer_1], properties {pitchMax=1.5, pitchMin=1.5}}
 {id=Modulator_1, children [WavePlayer_2], properties {pitchMax=1.5, pitchMin=1.5}}
 {id=Modulator_2, children [Mixer_0], properties {pitchMax=1.2, pitchMin=1}}
 {id=Random_0, children [Delay_0,Modulator_2], properties {weights [1,1]}}
 {id=WavePlayer_1, properties {bLooping=true, soundWaveAssetPtr=/Game/Sounds/Fort_Prop_Sounds/Exterior/PowerPlant/Transformer_Large_Buzz_01}}
 {id=WavePlayer_2, properties {bLooping=true, soundWaveAssetPtr=&/Spot_Sadium_Light_Buzz_Loop}}
]
properties {
 attenuationSettings=/Game/Sounds/Helios/Environment_Actors/NeonLightSignBuzz/Spot_Neon_Light_Buzz_Sign_Loop_Att
 bOverrideConcurrency=true
 maxDistance=3250
 pitchMultiplier=1.2
 soundClassObject=&Fort_Ambient_Outdoors_Athena
 volumeMultiplier=0.501
 attenuationOverrides {
  bEnableFocusInterpolation=true
  bEnableListenerFocus=true
  bEnableOcclusion=true
  dBAttenuationAtMax=-40
  distanceAlgorithm=Logarithmic
  falloffDistance=1500
  focusAzimuth=15
  focusVolumeAttenuation=1.1
  nonFocusAzimuth=45
  nonFocusVolumeAttenuation=0.4
  nonSpatializedRadiusStart=100
  occlusionTraceChannel=ECC_GameTraceChannel11
  occlusionVolumeAttenuation=0.7
  stereoSpread=2000
  attenuationShapeExtents {x=100}
 }
 busSends [{audioBus=null, minSendLevel=1, sendLevel=1, soundSourceBus=&AmbientAudio_Exterior_SourceBus}]
 concurrencyOverrides {maxCount=2, voiceStealReleaseTime=1}
 soundSubmixSends [{
   maxSendDistance=2000
   minSendDistance=500
   minSendLevel=0.4
   soundSubmix=&Fortnite_Reverb_Default
   customSendLevelCurve {editorCurveData {keys [<at 0 0.4>,<at 1 1>]}}
  }]
}
